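# Niagara system for the Dagger vehicle cosmetic drift trail, layering effects at the wheels.
# Burst-spawned cone, spinning dagger and wheel-wrap meshes follow skeletal-mesh sockets chosen by particle partition, while a two-ribbon emitter streams fading, width-animated trails.
# System modules unpack vehicle flags, and each emitter masks its particles per tire through FN_DisableTireFX.
# Painted-variant color overrides recolor every layer, with user flags adjusting brightness and ribbon behavior for painted and frontend-preview states.
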NiagaraSystem /VehicleCosmetics/DriftTrail/DriftTrail_Dagger/FX/NS_Trail_Dagger_MP
dependencies {
 usedDataInterfaces [
  &NiagaraDataInterfaceArrayFloat3
  &NiagaraDataInterfaceCurve
  &NiagaraDataInterfaceColorCurve
  &NiagaraDataInterfaceSkeletalMesh
 ]
 usedDynamicInputs []
 usedModules [
  &self:SystemSpawnScript.SystemScriptSource.SystemScriptGraph.NiagaraNodeAssignment_0.SetVariables_#1
  &SystemState
  &FN_UnpackVehicleFlags
  /Game/Effects/Niagara/Modules/Vehicle/FN_UpdateVehicleAttribs
  /Game/Effects/Niagara/Modules/Vehicle/FN_VehicleFlags_Cheatcodes
 ]
}
emitters {
 NE_VC_DriftTrail_BigCone {
  emitterUpdateScript [
   &EmitterState(
    Inactive Response=::NewEnumerator1
    Life Cycle Mode=::NewEnumerator1
    Loop Behavior=::NewEnumerator1
    Loop Duration Mode=::NewEnumerator1
    UseLoopDelay=false
   )
   &SpawnBurst_Instantaneous(
    Spawn Count=&MakeCustomIntFromBool(Bool=$User.bIsFrontendPreview, True INT=4, False INT=4)
    Spawn Group=1
    Use Spawn Probability=false
    UseLoopCountLimit=false
   ) {module=SpawnBurst_Instantaneous001}
   &SpawnBurst_Instantaneous(
    Spawn Count=&MakeCustomIntFromBool(Bool=$User.bIsFrontendPreview, True INT=4, False INT=4)
    Use Spawn Probability=false
    UseLoopCountLimit=false
   )
   &SpawnRate(
    SpawnRate=/VehicleCosmetics/FX/Niagara/Modules/VC_Mask_Floats(
     FrontEnd Value:NiagaraFloat=200
     blsFrontEndPreview:NiagaraBool=false
     SpeedMin Value:NiagaraFloat=120
     SpeedMax Value:NiagaraFloat=180
     Speed Alpha:NiagaraFloat=$System.Speed_Alpha
    )
    Use Spawn Probability=false
   ) {enabled=false}
  ]
  particleSpawnScript [
   &VC_PaintedVariantsOverride(
    Black=<rgba 0.012 0.012 0.012 7>
    BurntSienna=<rgba 0.07 0.0212 0.0035 4>
    Cobalt=<rgba 0.012 0.012 0.6 2>
    Crimson=<rgba 0.25 0.0375 0.0871 4>
    DefaultColor=<rgba 1 0.2625 0.25 1>
    ForestGreen=<rgba 0 0.28 0.0141 2>
    Grey=<rgba 0.1 0.1 0.1 2.5>
    Lime=<rgba 0.1833 0.5 0.025 2>
    Orange=<rgba 0.5 0.12 0.025 2>
    Pink=<rgba 0.8 0.12 0.5733 2>
    Purple=<rgba 0.1133 0 0.6795 8>
    R1 BurntSienna:LinearColor=<rgba 0.3 0.091 0.015 1>
    R1 Cobalt:LinearColor=<rgba 0 0 1.5 1.5>
    R1 Crimson:LinearColor=<rgba 2 0.04 0.1053 1>
    R1 Grey:LinearColor=<rgba 0.2 0.2 0.2 1.5>
    R1 Lime:LinearColor=<rgba 0.3333 1 0 1>
    R1 Orange:LinearColor=<rgba 1.5 0.3125 0.075 1>
    R1 Pink:LinearColor=<rgba 2 0.2396 1.3598 1>
    R1 Purple:LinearColor=<rgba 0.4167 0 1 1>
    R1 Saffron:LinearColor=<rgba 1 0.9167 0 1>
    R1 SkyBlue:LinearColor=<rgba 0 0.375 1.5 1>
    Saffron=<rgba 0.275 0.3 0 3>
    SkyBlue=<rgba 0 0.2 0.8 4>
    TitaniumWhite=<rgba 0.5 0.5 0.5 2>
    Use R1 Painted Overrides=true
   )
   /SetVariables(Particles.RENDERTAG:NiagaraInt32=&SpawnGroup()) {isSetParameters=true}
   &InitializeParticle(
    Color=$Particles.Color
    Lifetime=0.5
    Mesh Scale=/VehicleCosmetics/FX/Niagara/Modules/MergeEmitters/VC_VectorBySpawnGroup(
     SpawnGroup or ExecIndex:NiagaraInt32=0
     CustomInt:NiagaraInt32=0
     MaximumSpawnGroups:NiagaraInt32=2
     Vector Group 0:Vector3f=<0.31 0.069 0.069>
     Vector Group 1:Vector3f=<0.0775 0.03 0.03>
     Vector Group 2:Vector3f=<0 0 0>
     Vector Group 3:Vector3f=<0 0 0>
     Vector Group 4:Vector3f=<0 0 0>
     Vector Group 5:Vector3f=<0 0 0>
    )
    Mesh Scale Mode=::NewEnumerator4
    Sprite Rotation Mode=::NewEnumerator2
    Sprite Size Mode=::NewEnumerator3
    UsePositionOffset=false
    Write Material Random=true
   )
   &PartitionParticles()
   &SkeletalMeshLocation(
    Bone / Socket Index=&Add_Integer(A=$Particles.PartitionParticles.Partition, B=4)
    Bone / Socket Sampling Mode=::NewEnumerator5
    Interpolate Mesh Sampling=false
    MaskSpawnGroup=false
    Mesh Sampling Type=::NewEnumerator6
    Sampled Position Offset=&AddVectorToPosition(
     Position=&VC_VectorBySpawnGroup(
      SpawnGroup or ExecIndex:NiagaraInt32=0
      CustomInt:NiagaraInt32=0
      MaximumSpawnGroups:NiagaraInt32=2
      Vector Group 0:Vector3f=<-5 0 0>
      Vector Group 1:Vector3f=<-13 0 0>
      Vector Group 2:Vector3f=<0 0 0>
      Vector Group 3:Vector3f=<0 0 0>
      Vector Group 4:Vector3f=<0 0 0>
      Vector Group 5:Vector3f=<0 0 0>
     )
     Vector=$User.offset
    )
    Skeletal Mesh=$System.SkeletalMesh
   )
   &AddVelocity(Velocity=<1000 0 0>, Velocity Speed Scale=-1) {enabled=false}
   &InitialMeshOrientation(Mesh Orientation Mode=::NewEnumerator2, Orientation Coordinate Space=::Local, Orientation Vector=<-1 0 0>, Use Rotation Vector=false)
   &ApplyInitialForces()
  ]
  particleUpdateScript [
   &FN_DisableTireFX(
    Mask Index Attribute:NiagaraInt32=$Particles.PartitionParticles.Partition
    Mask Index FL=2
    Mask Index FR=3
    Mask Index RL=0
    Mask Index RR=1
   )
   &ParticleState(Kill Particles When Lifetime Has Elapsed=false, Loop Particles Lifetime=true)
   &ScaleColor(Scale Alpha=2, ScaleA=true, ScaleRGB=true)
   &SolveForcesAndVelocity(Clamp Velocity=false, Limit Acceleration=false)
   &SkeletalMeshLocation(
    Bone / Socket Index=&Add_Integer(A=$Particles.PartitionParticles.Partition, B=4)
    Bone / Socket Sampling Mode=::NewEnumerator5
    MaskSpawnGroup=false
    Mesh Sampling Type=::NewEnumerator6
    Sampled Position Offset=&AddVectorToPosition(
     Position=&VC_VectorBySpawnGroup(
      SpawnGroup or ExecIndex:NiagaraInt32=0
      CustomInt:NiagaraInt32=0
      MaximumSpawnGroups:NiagaraInt32=2
      Vector Group 0:Vector3f=<-5 0 0>
      Vector Group 1:Vector3f=<-13 0 0>
      Vector Group 2:Vector3f=<0 0 0>
      Vector Group 3:Vector3f=<0 0 0>
      Vector Group 4:Vector3f=<0 0 0>
      Vector Group 5:Vector3f=<0 0 0>
     )
     Vector=$User.offset
    )
    Skeletal Mesh=$System.SkeletalMesh
   ) {module=SkeletalMeshLocation001}
  ]
  rendererClasses [&NiagaraMeshRendererProperties, &NiagaraSpriteRendererProperties]
  renderers [
   {rendererClass=&NiagaraMeshRendererProperties, rendererIndex=0}
   {rendererClass=&NiagaraMeshRendererProperties, rendererIndex=1}
   {rendererClass=&NiagaraSpriteRendererProperties, rendererIndex=2}
  ]
 }
 NE_VC_DriftTrail_DAGGER {
  emitterUpdateScript [
   &EmitterState(
    Inactive Response=::NewEnumerator1
    Life Cycle Mode=::NewEnumerator1
    Loop Behavior=::NewEnumerator1
    Loop Duration Mode=::NewEnumerator1
    UseLoopDelay=false
   )
   &SpawnBurst_Instantaneous(
    Spawn Count=&MakeCustomIntFromBool(Bool=$User.bIsFrontendPreview, True INT=1, False INT=0)
    Spawn Group=1
    Use Spawn Probability=false
    UseLoopCountLimit=false
   ) {enabled=false, module=SpawnBurst_Instantaneous002}
   &SpawnBurst_Instantaneous(
    Spawn Count=&MakeCustomIntFromBool(Bool=$User.bIsFrontendPreview, True INT=4, False INT=4)
    Spawn Group=1
    Use Spawn Probability=false
    UseLoopCountLimit=false
   ) {module=SpawnBurst_Instantaneous001}
   &SpawnRate(
    SpawnRate=&VC_Mask_Floats(
     FrontEnd Value:NiagaraFloat=200
     blsFrontEndPreview:NiagaraBool=false
     SpeedMin Value:NiagaraFloat=120
     SpeedMax Value:NiagaraFloat=180
     Speed Alpha:NiagaraFloat=$System.Speed_Alpha
    )
    Use Spawn Probability=false
   ) {enabled=false}
  ]
  particleSpawnScript [
   &VC_PaintedVariantsOverride(
    Black=<rgba 0.015 0.015 0.015 1>
    BurntSienna=<rgba 0.1 0.0303 0.005 1>
    Cobalt=<rgba 0 0.0167 1 1>
    Crimson=<rgba 0.8 0 0.02 1>
    DefaultColor=<rgba 2 1.1245 0.4 1>
    ForestGreen=<rgba 0 0.47 0.0237 1>
    Grey=<rgba 0.065 0.065 0.065 1>
    Lime=<rgba 0.1458 0.35 0 1>
    Orange=<rgba 1.15 0.0069 0.0069 1>
    Pink=<rgba 0.7 0 0.2917 1>
    Purple=<rgba 0.175 0 0.3 1>
    Saffron=<rgba 0.5 0.2917 0 1>
    SkyBlue=<rgba 0 0.25 1 1>
    TitaniumWhite=<rgba 0 0 0 1>
   )
   &InitializeParticle(
    Color=$Particles.Color
    Mesh Scale Mode=::NewEnumerator3
    Sprite Rotation Mode=::NewEnumerator2
    Sprite Size Mode=::NewEnumerator3
    UsePositionOffset=false
    Write Material Random=true
   )
   &PartitionParticles(Particles Per Partition=1, Partition Mode:ENiagara_PartitionMode=::NewEnumerator1)
   &SkeletalMeshLocation(
    Bone / Socket Index=&Add_Integer(A=$Particles.PartitionParticles.Partition, B=4)
    Bone / Socket Sampling Mode=::NewEnumerator5
    Interpolate Mesh Sampling=false
    MaskSpawnGroup=false
    Mesh Sampling Type=::NewEnumerator6
    Sampled Position Offset=&AddVectorToPosition(Position=<-43 0 3>, Vector=$User.offset)
    Skeletal Mesh=$System.SkeletalMesh
   )
   &AddVelocity(Velocity=&VC_Mask_Vectors(
     FrontEnd Value:Vector3f=<-1500 0 0>
     blsFrontEndPreview:NiagaraBool=false
     SpeedMin Value:Vector3f=<-800 0 0>
     SpeedMax Value:Vector3f=<-1 0 0>
     Speed Alpha:NiagaraFloat=$System.Speed_Alpha
    )) {enabled=false}
   &InitialMeshOrientation(Mesh Orientation Mode=::NewEnumerator2, Orientation Coordinate Space=::Local, Rotation=<0 0.01 0>, Use Rotation Vector=true)
   &CameraOffset(Camera Offset Amount=20)
  ]
  particleUpdateScript [
   &FN_DisableTireFX(
    Mask Index Attribute:NiagaraInt32=$Particles.PartitionParticles.Partition
    Mask Index FL=2
    Mask Index FR=3
    Mask Index RL=0
    Mask Index RR=1
   )
   &ParticleState(Lifetime:NiagaraFloat=$Emitter.LoopedAge)
   &UpdateMeshOrientation(
    Rotation Rate=&SelectFloatFromArray(
     Direct Array Index=&Add_Integer(A=$Particles.PartitionParticles.Partition, B=0)
     Float Selection Array=DataInterface() {propertyValues {FloatData [1.2,-1.5]}}
    )
    Rotation Vector=&Multiply_VectorByFloat(
     Vector=<1 0 0>
     Float=&SelectFloatFromArray(Direct Array Index=$Particles.PartitionParticles.Partition, Float Selection Array=DataInterface() {propertyValues {FloatData [1.5,1]}})
    )
   ) {module=UpdateMeshOrientation002}
   &ScaleColor(
    Scale RGB=&Multiply_VectorByFloat(Vector=<1 1 1>, Float=&MakeCustomFloatFromBool(Bool=$User.bIsPainted, True Float=6.5, False Float=1))
    ScaleA=true
    ScaleRGB=true
   )
   &SolveForcesAndVelocity(Clamp Velocity=false, Limit Acceleration=false)
   &SkeletalMeshLocation(
    Bone / Socket Index=&Add_Integer(A=$Particles.PartitionParticles.Partition, B=4)
    Bone / Socket Sampling Mode=::NewEnumerator5
    MaskSpawnGroup=false
    Mesh Sampling Type=::NewEnumerator6
    Orientation Sampling=::NewEnumerator1
    Sampled Position Offset=&AddVectorToPosition(Position=<-43 0 3>, Vector=$User.offset)
    Skeletal Mesh=$System.SkeletalMesh
   ) {module=SkeletalMeshLocation001}
  ]
  rendererClasses [&NiagaraMeshRendererProperties, &NiagaraSpriteRendererProperties]
  renderers [
   {rendererClass=&NiagaraMeshRendererProperties, rendererIndex=0}
   {rendererClass=&NiagaraSpriteRendererProperties, rendererIndex=1}
  ]
 }
 NE_VC_DriftTrail_Ribbon_MAIN002 {
  emitterUpdateScript [
   &EmitterState(
    Inactive Response=::NewEnumerator1
    Life Cycle Mode=::NewEnumerator1
    Loop Duration=1
    Recalculate Duration Each Loop=true
    UseLoopDelay=false
   )
   &VC_RateQualityHandler(Epic Spawn Rate=90, High Spawn Rate=60, Low Spawn Rate=30, Medium Spawn Rate=45, Use Custom Rate?=true)
   &VC_SpawnRibbon_v1(
    Number of Ribbons=2
    Spawn Time=-1
    SpawnRate=&MakeCustomFloatFromBool(Bool=$User.bIsFrontendPreview, True Float=35, False Float=35)
   )
  ]
  particleSpawnScript [
   &VC_PaintedVariantsOverride(
    Black=<rgba 0.012 0.012 0.012 1>
    BurntSienna=<rgba 0.03 0.0116 0.006 0.8>
    Cobalt=<rgba 0.0074 0.005 0.1 1>
    Crimson=<rgba 0.15 0.0075 0.017 0.9>
    ForestGreen=<rgba 0 0.07 0.0035 1>
    Grey=<rgba 0.02 0.02 0.02 0.8>
    Lime=<rgba 0.0167 0.05 0 1>
    Orange=<rgba 0.08 0.028 0 1>
    Pink=<rgba 0.06 0.021 0.06 1>
    Purple=<rgba 0.0375 0 0.15 1>
    Saffron=<rgba 0.0583 0.07 0 1>
    SkyBlue=<rgba 0 0.034 0.12 1>
    TitaniumWhite=<rgba 0.096 0.1044 0.12 1>
   )
   &VC_InitializeRibbon_v1(
    Color=$Particles.Color
    Custom Offset=true
    Enable Skeletal Position=true
    Local Vector Offset=&AddVectorToPosition(Position=<17 0 0>, Vector=$User.offset)
    Offset Amount=1
    Ribbon Width=60
   )
   &AddVelocity(Velocity=/Niagara/DynamicInputs/TypeConversions/MakeCustomPositionFromBool(Bool:NiagaraBool=$User.bIsFrontendPreview, True Position:NiagaraPosition=<-2000 0 80>, False Position:NiagaraPosition=<-2000 0 120>)) {module=AddVelocity001}
   /VehicleCosmetics/FX/Niagara/Modules/MergeEmitters/VC_ExtraBindingColors(
    ExtraColors=2
    R1 Color:LinearColor=&MakeCustomLinearColorFromBool(Bool=$User.bIsPainted, True Color=$User.CORECOLOR, False Color=<rgba 0.898 0.35 1 1>)
    R2 Color:LinearColor=&MakeCustomLinearColorFromBool(Bool=$User.bIsPainted, True Color=$User.SPARKSCOLOR, False Color=<rgba 1 0.325 0.1 1>)
   )
  ]
  particleUpdateScript [
   &FN_DisableTireFX(Mask Index Attribute:NiagaraInt32=$Particles.SpawnGroup, Mask Index FL=2, Mask Index FR=3, Mask Index RL=0, Mask Index RR=1)
   &InheritSourceMovement(
    Applied Position Scale=&Multiply_VectorByFloat(
     Vector=<1 1 1>
     Float=&FloatFromCurve(FloatCurve=DataInterface() {propertyValues {Curve {keys [{interpMode=RCIM_Cubic, tangentMode=RCTM_User, time=0.2, value=1}, {interpMode=RCIM_Cubic, tangentMode=RCTM_User, time=1, value=0}]}}})
    )
    ApplyForce=false
    ApplyPosition=true
    ApplyPositionFalloff=false
   )
   &ParticleState()
   &SolveForcesAndVelocity(Clamp Velocity=false, Limit Acceleration=false)
   &DynamicMaterialParameters(
    Index 0 Param 1=&FloatFromCurve(FloatCurve=DataInterface() {propertyValues {Curve {keys [<at 0 1>,<at 1 0.5>]}}})
    Index 1 Param 1=&MakeCustomFloatFromBool(
     Bool=$User.bIsPainted
     True Float=&FloatFromCurve(
      FloatCurve=DataInterface() {propertyValues {Curve {keys [{arriveTangent=-7.7015, interpMode=RCIM_Cubic, leaveTangent=-7.7015, tangentMode=RCTM_User, time=0.1, value=1}, <at 0.5 0.1 RCIM_Cubic>]}}}
      Scale Curve=&MakeCustomFloatFromBool(Bool=$User.bIsFrontendPreview, True Float=10, False Float=5)
     )
     False Float=&MakeCustomFloatFromBool(Bool=$User.bIsFrontendPreview, True Float=1, False Float=2)
    )
    Index 1 Param 2=0.5
    Index 1 Param 3=0.5
    Index 1 Param 4=-0.1
    Index 2 Param 1=1
    Module.Index0Param0Write=true
    Module.Index0Param1Write=true
    Module.Index0Param2Write=true
    Module.Index0Param3Write=true
    Module.Index1Param0Write=true
    Module.Index1Param1Write=true
    Module.Index1Param2Write=true
    Module.Index1Param3Write=true
    Module.Index2Param0Write=true
    Module.Index2Param1Write=true
    Module.Index2Param2Write=true
    Module.Index2Param3Write=true
    Write Parameter Index 1=true
    Write Parameter Index 2=true
   )
   &ScaleColor(
    Scale Alpha=&FloatFromCurve(
     FloatCurve=DataInterface() {propertyValues {Curve {keys [<at 0 1 RCIM_Cubic>, {arriveTangentWeight=0.6633, interpMode=RCIM_Cubic, tangentMode=RCTM_User, time=1, value=0}]}}}
     Scale Curve=&MakeCustomFloatFromBool(Bool=$User.bIsFrontendPreview, True Float=4, False Float=6)
    )
    Scale RGB=&Multiply_VectorByFloat(
     Vector=<1 1 1>
     Float=&FloatFromCurve(
      FloatCurve=DataInterface() {propertyValues {Curve {keys [{arriveTangent=-0.1571, interpMode=RCIM_Cubic, leaveTangent=-0.1571, tangentMode=RCTM_User, time=0, value=2}, <at 0.7 0 RCIM_Cubic>]}}}
      Scale Curve=&MakeCustomFloatFromBool(
       Bool=$User.bIsFrontendPreview
       True Float=5
       False Float=&MakeCustomFloatFromBool(Bool=$User.bIsPainted, True Float=15, False Float=5)
      )
     )
    )
    ScaleA=true
    ScaleRGB=true
   )
   &Color(Color=&MakeCustomLinearColorFromBool(
     Bool=$User.bIsPainted
     True Color=$Particles.Color
     False Color=&ColorFromCurve(ColorCurve=DataInterface() {propertyValues {
        AlphaCurve {keys [<at 0 1>, {arriveTangent=1.4362, interpMode=RCIM_Cubic, leaveTangent=1.4362, time=0.2534, value=5}, <at 0.6963 2>, <at 0.8851 0 RCIM_Cubic>]}
        BlueCurve {keys [<at 0 1.2 RCIM_Cubic>,<at 0.4566 0.2604>]}
        GreenCurve {keys [<at 0 0.3 RCIM_Cubic>,<at 0.4566 0.0261>]}
        RedCurve {keys [<at 0 1.5 RCIM_Cubic>,<at 0.4566 0.0861>]}
       }})
    ))
   &ScaleRibbonWidth(Ribbon Width Scale=&FloatFromCurve(
     FloatCurve=DataInterface() {propertyValues {Curve {keys [
         {arriveTangent=1.8526, interpMode=RCIM_Cubic, leaveTangent=1.8526, leaveTangentWeight=0.3476, tangentMode=RCTM_User, time=0.1, value=0.8}
         {arriveTangentWeight=0.3333, interpMode=RCIM_Cubic, tangentMode=RCTM_User, time=0.9, value=3}
        ]}}}
     Scale Curve=&SelectFloatFromArray(Direct Array Index=$Particles.SpawnGroup, Float Selection Array=DataInterface() {propertyValues {FloatData [1,-1]}})
    ))
   /SetVariables(
    Particles.R1 Color:LinearColor=&VC_DynamicScaleColor(
     Color to Scale:LinearColor=$Particles.R1 Color
     RGB Scale:Vector3f=&Multiply_VectorByFloat(
      Vector=<1 1 1>
      Float=&MakeCustomFloatFromBool(
       Bool=$User.bIsPainted
       True Float=&FloatFromCurve(
        FloatCurve=DataInterface() {propertyValues {Curve {keys [{interpMode=RCIM_Cubic, tangentMode=RCTM_User, time=0, value=1}, {interpMode=RCIM_Cubic, tangentMode=RCTM_User, time=0.7, value=0}]}}}
        Scale Curve=&MakeCustomFloatFromBool(Bool=$User.bIsFrontendPreview, True Float=3.1, False Float=3.2)
       )
       False Float=&FloatFromCurve(
        FloatCurve=DataInterface() {propertyValues {Curve {keys [
            {arriveTangent=-1.1264, interpMode=RCIM_Cubic, leaveTangent=-1.1264, tangentMode=RCTM_User, time=0, value=1}
            {interpMode=RCIM_Cubic, tangentMode=RCTM_User, time=0.75, value=0}
           ]}}}
        Scale Curve=2.5
       )
      )
     )
     Alpha Scale:NiagaraFloat=&FloatFromCurve(FloatCurve=DataInterface() {propertyValues {Curve {keys [
          {interpMode=RCIM_Cubic, leaveTangentWeight=0.2667, tangentMode=RCTM_User, time=0, value=1}
          {arriveTangentWeight=0.4939, interpMode=RCIM_Cubic, tangentMode=RCTM_User, time=1, value=0}
         ]}}})
    )
    Particles.R2 Color:LinearColor=&VC_DynamicScaleColor(
     Color to Scale:LinearColor=$Particles.R2 Color
     RGB Scale:Vector3f=&VectorFromFloat(Value=&MakeCustomFloatFromBool(
       Bool=$User.bIsPainted
       True Float=&FloatFromCurve(FloatCurve=DataInterface() {propertyValues {Curve {keys [<at 0 1 RCIM_Cubic>,<at 0.8 0 RCIM_Cubic>]}}}, Scale Curve=3.3)
       False Float=2
      ))
     Alpha Scale:NiagaraFloat=&MakeCustomFloatFromBool(
      Bool=$User.bIsPainted
      True Float=&FloatFromCurve(FloatCurve=DataInterface() {propertyValues {Curve {keys [<at 0 0>,<at 0.1 1>,<at 0.6 0>]}}}, Scale Curve=1.2)
      False Float=&FloatFromCurve(FloatCurve=DataInterface() {propertyValues {Curve {keys [<at 0 0>,<at 0.1 1>,<at 0.6 0>]}}}, Scale Curve=1.2)
     )
    )
    Particles.RIBBONWIDTH3:NiagaraFloat=&Multiply_Float(
     A=32
     B=&FloatFromCurve(
      FloatCurve=DataInterface() {propertyValues {Curve {keys [
          {interpMode=RCIM_Cubic, leaveTangentWeight=0.0834, tangentMode=RCTM_User, time=0, value=0}
          {arriveTangent=4, interpMode=RCIM_Cubic, leaveTangent=7.3067, tangentMode=RCTM_Break, time=0.07, value=0.5}
          {arriveTangentWeight=0.0707, interpMode=RCIM_Cubic, leaveTangentWeight=0.4007, tangentMode=RCTM_User, time=0.25, value=1}
          <at 1 1>
         ]}}}
      Scale Curve=-1
     )
    )
    Particles.RIBBONWIDTH?2=26
   ) {isSetParameters=true}
   &VC_UpdateRibbon_v1(Enable Skeletal Position=true)
  ]
  rendererClasses [&NiagaraRibbonRendererProperties]
  renderers [
   {rendererClass=&NiagaraRibbonRendererProperties}
   {rendererClass=&NiagaraRibbonRendererProperties, rendererIndex=1}
   {rendererClass=&NiagaraRibbonRendererProperties, rendererIndex=2}
  ]
 }
 NE_VC_DriftTrail_WheelWrap_Parent {
  emitterUpdateScript [
   &EmitterState(Inactive Response=::NewEnumerator1, Life Cycle Mode=::NewEnumerator1, Loop Duration=1, UseLoopDelay=false)
   &SpawnRate(SpawnRate=4, Use Spawn Probability=false) {enabled=false}
   &SpawnBurst_Instantaneous(Spawn Count=4, Use Spawn Probability=false, UseLoopCountLimit=false)
  ]
  particleSpawnScript [
   &VC_PaintedVariantsOverride(
    Black=<rgba 0.005 0.005 0.005 10>
    BurntSienna=<rgba 0.15 0.033 0.015 1>
    Cobalt=<rgba 0 0 0.25 2>
    Crimson=<rgba 0.4635 0 0.0193 1>
    DefaultColor=<rgba 1 0.3033 0.05 0.25>
    ForestGreen=<rgba 0 0.5 0.0253 0.5>
    Grey=<rgba 0.05 0.05 0.05 1>
    Lime=<rgba 0.0981 0.5885 0 1>
    Orange=<rgba 0.6724 0.0785 0 1>
    Pink=<rgba 1 0.1 0.325 0.5>
    Purple=<rgba 0.09 0 0.6 1>
    Saffron=<rgba 0.7 0.7 0 0.25>
    SkyBlue=<rgba 0 0.2235 0.5365 1>
    TitaniumWhite=<rgba 0.75 0.75 0.75 0.5>
   )
   &InitializeParticle(
    Color=$Particles.Color
    Mesh Scale Mode=::NewEnumerator4
    Sprite Rotation Mode=::NewEnumerator2
    Sprite Size Mode=::NewEnumerator3
    UsePositionOffset=false
    Write Material Random=true
   )
   &PartitionParticles()
   &SkeletalMeshLocation(
    Bone / Socket Index=&Add_Integer(A=$Particles.PartitionParticles.Partition, B=4)
    Bone / Socket Sampling Mode=::NewEnumerator5
    Interpolate Mesh Sampling=false
    MaskSpawnGroup=false
    Mesh Sampling Type=::NewEnumerator6
    Orientation Sampling=::NewEnumerator1
    Sampled Position Offset=&SelectVectorFromArray(
     Direct Array Index=$Particles.PartitionParticles.Partition
     Vector Selection Array=DataInterface() {propertyValues {FloatData [<0 0 0>, <0 0 0>]}}
    )
    Skeletal Mesh=$System.SkeletalMesh
   )
   &InitialMeshOrientation(Mesh Orientation Mode=::NewEnumerator2, Use Rotation Vector=true)
  ]
  particleUpdateScript [
   &FN_DisableTireFX(
    Mask Index Attribute:NiagaraInt32=$Particles.PartitionParticles.Partition
    Mask Index FL=2
    Mask Index FR=3
    Mask Index RL=0
    Mask Index RR=1
   )
   &VC_WheelWrapScale(Wheel Index Attribute:NiagaraInt32=&Add_Integer(A=$Particles.PartitionParticles.Partition, B=2))
   &DynamicMaterialParameters(
    Index 0 Param 1=1
    Index 1 Param 1=1
    Index 1 Param 2=1
    Index 2 Param 1=&FloatFromCurve(FloatCurve=DataInterface() {propertyValues {Curve {keys [{interpMode=RCIM_Cubic, tangentMode=RCTM_User, time=0, value=1}, {interpMode=RCIM_Cubic, tangentMode=RCTM_User, time=1, value=1}]}}})
    Module.Index0Param0Write=true
    Module.Index0Param1Write=true
    Module.Index0Param2Write=true
    Module.Index0Param3Write=true
    Module.Index1Param0Write=true
    Module.Index1Param1Write=true
    Module.Index1Param2Write=true
    Module.Index1Param3Write=true
    Module.Index2Param0Write=true
    Module.Index2Param1Write=true
    Module.Index2Param2Write=true
    Module.Index2Param3Write=true
    Write Parameter Index 1=true
    Write Parameter Index 2=true
   )
   &ScaleColor(
    Scale Alpha=2
    Scale RGB=&Multiply_VectorByFloat(
     Vector=<1 1 1>
     Float=&FloatFromCurve(FloatCurve=DataInterface() {propertyValues {Curve {keys [<at 0 1 RCIM_Cubic>,<at 1 1 RCIM_Cubic>]}}}, Scale Curve=15)
    )
    ScaleA=true
    ScaleRGB=true
   )
   &Color(Color=&MakeCustomLinearColorFromBool(
     Bool=$User.bIsPainted
     True Color=$Particles.Color
     False Color=&ColorFromCurve(ColorCurve=DataInterface() {propertyValues {
        AlphaCurve {keys [<at 0 1 RCIM_Cubic>,<at 1 1 RCIM_Cubic>]}
        BlueCurve {keys [<at 0 0 RCIM_Cubic>]}
        GreenCurve {keys [<at 0 0.2158 RCIM_Cubic>]}
        RedCurve {keys [<at 0 1 RCIM_Cubic>]}
       }})
    )) {enabled=false}
   &ParticleState()
   &UpdateMeshOrientation(Rotation Coordinate Space=::NewEnumerator2, Rotation Vector=<0 1 0>) {enabled=false}
   &SkeletalMeshLocation(
    Bone / Socket Index=&Add_Integer(A=$Particles.PartitionParticles.Partition, B=4)
    Bone / Socket Sampling Mode=::NewEnumerator5
    MaskSpawnGroup=false
    Mesh Sampling Type=::NewEnumerator6
    Skeletal Mesh=$System.SkeletalMesh
   ) {module=SkeletalMeshLocation001}
   &SolveForcesAndVelocity(Clamp Velocity=false, Limit Acceleration=false)
  ]
  rendererClasses [&NiagaraMeshRendererProperties]
  renderers [{rendererClass=&NiagaraMeshRendererProperties, rendererIndex=0}]
 }
}
summary {
 systemName=NS_Trail_Dagger_MP
 emitters [
  {
   emitterName=NE_VC_DriftTrail_BigCone
   rendererClasses [&NiagaraMeshRendererProperties, &NiagaraSpriteRendererProperties]
  }
  {
   emitterName=NE_VC_DriftTrail_DAGGER
   rendererClasses [&NiagaraMeshRendererProperties, &NiagaraSpriteRendererProperties]
  }
  {emitterName=NE_VC_DriftTrail_WheelWrap_Parent, rendererClasses [&NiagaraMeshRendererProperties]}
  {emitterName=NE_VC_DriftTrail_Ribbon_MAIN002, rendererClasses [&NiagaraRibbonRendererProperties]}
 ]
 userVariables [
  User.bIsFrontendPreview {type=NiagaraBool, defaultValue=true}
  User.bIsPainted {type=NiagaraBool}
  User.ColorVariantID {type=NiagaraInt32}
  User.CORECOLOR {type=LinearColor, description="", defaultValue {struct=LinearColor, value=<rgba 1 1 1 1>}}
  User.offset {type=Vector3f, description="offset from wheels(set through cosmetic template)", defaultValue {struct=Vector3f, value=<0 0 0>}}
  User.PaintedColor {type=LinearColor, description="", defaultValue {struct=LinearColor, value=<rgba 1 1 1 1>}}
  User.PaintedScalar {type=NiagaraFloat, defaultValue=1, description=""}
  User.SPARKSCOLOR {type=LinearColor, description="", defaultValue {struct=LinearColor, value=<rgba 1 1 1 1>}}
  User.VehicleFlags {type=NiagaraInt32, defaultValue=806289408}
  User.Wheel_Radius_Array {
   type=NiagaraDataInterfaceArrayFloat
   defaultValue {
    struct=NiagaraExt_VariableValue_DataInterface
    value {
     dataInterface=&self:NiagaraDataInterfaceArrayFloat_0
     dataInterfaceClass=&NiagaraDataInterfaceArrayFloat
    }
   }
  }
  User.Wheel_Width_Array {
   type=NiagaraDataInterfaceArrayFloat
   defaultValue {
    struct=NiagaraExt_VariableValue_DataInterface
    value {
     dataInterface=&self:NiagaraDataInterfaceArrayFloat_1
     dataInterfaceClass=&NiagaraDataInterfaceArrayFloat
    }
   }
  }
  User.WheelSize {type=NiagaraFloat, defaultValue=0.1, description=""}
 ]
}
