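# Translucent, two-sided unlit material for Niagara mesh particles that draws the airwave effect in the Dagger drift trail vehicle cosmetic.
# A panning wind mask, faded along V, is shared by named reroute.
# Particle color tints it for emissive via MF_FX_EmissiveAdjustments, while its sharpened power curve times particle alpha gives opacity.
# A time-driven sine of world X, controlled by Speed, ripples vertices along their normals, weighted by V.
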Material /VehicleCosmetics/DriftTrail/DriftTrail_Dagger/Materials/M_Trail_Airwavemat
graph {
 properties {
  bAutomaticallySetUsageInEditor=false
  bUsedWithNiagaraMeshParticles=true
  blendMode=BLEND_Translucent
  shadingModel=MSM_Unlit
  twoSided=true
  usedShadingModels=Unlit
 }
 nodes [
  n1 = NamedRerouteUsage()
  MaterialExpressionConstantClamp_7 = Saturate(None=Multiply(A=Power(Base=@n1) {constExponent=4}, B=ParticleColor().A))
  MaterialExpressionMeshEmitterVertexColor_4 = ParticleColor()
  MaterialFunctionCall(Emissive=Multiply(A=Multiply(A=@MaterialExpressionMeshEmitterVertexColor_4.RGB, B=@n1)) {constB=3}) {materialFunction=&MF_FX_EmissiveAdjustments}
  n2 = ScalarParameter() {defaultValue=5, parameterName=Speed}
  n3 = ComponentMask(None=Add(
    A=WorldPosition().XYZ
    B=MaterialFunctionCall(Speed=@n2) {materialFunction=&TimeWithSpeedVariable}.Time
   )) {r=true}
  MaterialExpressionWorldNormal_3 = VertexNormalWS()
  n4 = TextureCoordinate()
  MaterialExpressionConstantClamp_8 = Saturate(None=Multiply(A=ComponentMask(None=@n4) {g=true}) {constA=2, constB=2})
  Multiply(A=Multiply(A=Sine(None=@n3), B=@MaterialExpressionWorldNormal_3), B=@MaterialExpressionConstantClamp_8)
  n5 = Panner(Coordinate=TextureCoordinate() {vTiling=2}) {speedX=1, speedY=-3}
  n6 = OneMinus(None=ComponentMask(None=TextureCoordinate()) {g=true})
  n7 = Saturate(None=Multiply(A=Power(Base=@n6) {constExponent=4}) {constB=4})
  NamedRerouteDeclaration(None=Multiply(A=TextureSample(UVs=@n5) {texture=/DealershipSharedLibrary/FX/Textures/Gradient_Masks/T_VCVFX_Wind_A}.RGB, B=Multiply(A=@n7) {constB=0.8})) {name=Texture, nodeColor=<rgba 1 0.0234 0 1>}
 ]
}
