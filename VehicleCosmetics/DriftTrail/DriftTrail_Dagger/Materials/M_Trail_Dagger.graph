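# Unlit, translucent, two-sided material for Niagara mesh particles in the Dagger drift trail vehicle cosmetic.
# Particle color RGB feeds MF_FX_EmissiveAdjustments for the emissive look.
# A procedural drift vector, apparently for world position offset, blends time-driven sine waves per axis using sine or cosine alphas.
# Speed parameters drive the waves, per-axis and uniform multipliers scale the vector, and a constant offset parameter is added.
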Material /VehicleCosmetics/DriftTrail/DriftTrail_Dagger/Materials/M_Trail_Dagger
graph {
 properties {
  bAutomaticallySetUsageInEditor=false
  bUsedWithNiagaraMeshParticles=true
  blendMode=BLEND_Translucent
  shadingModel=MSM_Unlit
  twoSided=true
  usedShadingModels=Unlit
 }
 nodes [
  n1 = MaterialFunctionCall(Speed=ScalarParameter() {defaultValue=1.5, parameterName=SpeedDB}) {materialFunction=&TimeWithSpeedVariable}
  n2 = MaterialFunctionCall(Speed=ScalarParameter() {defaultValue=1, parameterName=SpeedAB}) {materialFunction=&TimeWithSpeedVariable}
  n3 = Multiply(
   A=LinearInterpolate(B=Sine(None=@n1.Time), Alpha=Sine(None=@n2.Time)) {constA=1}
   B=ScalarParameter() {defaultValue=1, parameterName=X_Drift_Mult}
  )
  n4 = MaterialFunctionCall(Speed=ScalarParameter() {defaultValue=1.5, parameterName=SpeedC}) {materialFunction=&TimeWithSpeedVariable}
  n5 = Multiply(A=ScalarParameter() {defaultValue=1, parameterName=SpeedB}, B=Add(A=Time(), B=ScalarParameter() {defaultValue=1, parameterName=TimeOffset}))
  n6 = Multiply(
   A=LinearInterpolate(A=Sine(None=@n4.Time), Alpha=Cosine(None=@n5)) {constB=-1}
   B=ScalarParameter() {defaultValue=1, parameterName=Z_Drift_Mult}
  )
  n7 = MaterialFunctionCall(Speed=ScalarParameter() {defaultValue=1.5, parameterName=SpeedD}) {materialFunction=&TimeWithSpeedVariable}
  n8 = MaterialFunctionCall(Speed=ScalarParameter() {defaultValue=1, parameterName=SpeedA}) {materialFunction=&TimeWithSpeedVariable}
  n9 = Multiply(
   A=LinearInterpolate(B=Sine(None=@n7.Time), Alpha=Sine(None=@n8.Time)) {constA=1}
   B=ScalarParameter() {defaultValue=1, parameterName=Y_Drift_Mult}
  )
  Add(
   A=Multiply(
    A=MaterialFunctionCall(X=@n3, Y=@n9, Z=@n6) {materialFunction=&MakeFloat3}.Result
    B=ScalarParameter() {defaultValue=5, parameterName=Uniform_Drift_Mult}
   )
   B=VectorParameter() {parameterName=WPO_OffsetB}.RGB
  )
  MaterialExpressionMeshEmitterVertexColor_5 = ParticleColor()
  MaterialFunctionCall(Emissive=@MaterialExpressionMeshEmitterVertexColor_5.RGB) {materialFunction=&MF_FX_EmissiveAdjustments}
 ]
}
