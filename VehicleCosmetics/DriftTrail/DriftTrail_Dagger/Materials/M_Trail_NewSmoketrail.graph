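# Unlit translucent smoke trail material for the Dagger drift trail, used on Niagara ribbons and mesh particles.
# Static switches select main ribbon, secondary trail or wind trail.
# Panned smoke and cloud textures, UV fades, particle color and dynamic parameters build the layers, with noise-distorted UVs for the secondary trail and two speed-line textures for another.
# Emissive passes through an adjustment function, while opacity gets depth fade and an optional wheel-wrap mask.
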Material /VehicleCosmetics/DriftTrail/DriftTrail_Dagger/Materials/M_Trail_NewSmoketrail
graph {
 properties {
  bAutomaticallySetUsageInEditor=false
  bUsedWithNiagaraMeshParticles=true
  bUsedWithNiagaraRibbons=true
  blendMode=BLEND_Translucent
  shadingModel=MSM_Unlit
  usedShadingModels=Unlit
 }
 nodes [
  Constant() {r=50}
  n1 = NamedRerouteUsage()
  n2 = NamedRerouteUsage()
  MaterialFunctionCall(Emissive=StaticSwitchParameter(True=NamedRerouteUsage(), False=StaticSwitchParameter(True=@n1, False=@n2) {parameterName=IS MAIN RIBBON?}) {parameterName=IS WINDTRAIL?}) {materialFunction=&MF_FX_EmissiveAdjustments}
  MaterialExpressionVertexColor_4 = ParticleColor()
  NamedRerouteDeclaration(None=@MaterialExpressionVertexColor_4.RGB) {name=Particles RGB, nodeColor=<rgba 0.0781 0 1 1>}
  NamedRerouteDeclaration(None=@MaterialExpressionVertexColor_4.A) {name=Particles Alpha, nodeColor=<rgba 0.2422 1 0 1>}
  n3 = Multiply(A=NamedRerouteUsage(), B=NamedRerouteUsage())
  n4 = Multiply(A=NamedRerouteUsage(), B=@n3)
  n5 = NamedRerouteUsage()
  n6 = NamedRerouteUsage()
  MaterialExpressionConstantClamp_9 = Saturate(None=Multiply(A=Multiply(A=@n4, B=@n5), B=Saturate(None=@n6)))
  NamedRerouteDeclaration(None=@MaterialExpressionConstantClamp_9) {name=Trail 2 Alpha, nodeColor=<rgba 0 1 0.7891 1>}
  n7 = Multiply(A=TextureCoordinate() {vTiling=0.2}, B=ScalarParameter() {defaultValue=1, parameterName=texcoord})
  n8 = MaterialFunctionCall(Speed=ScalarParameter() {defaultValue=1, parameterName=speed01}) {materialFunction=&TimeWithSpeedVariable}
  n9 = VectorParameter() {parameterName=SpeedTextures, channelNames {a=Speed Y Tex2, b=Speed X Tex2, g=Speed Y Tex1, r=Speed X Tex1}, defaultValue=<rgba -1 0 -3 0>}
  n10 = MaterialFunctionCall(X=@n9.Speed X Tex1, Y=@n9.Speed X Tex1) {materialFunction=&MakeFloat2}
  n11 = Multiply(A=TextureCoordinate(), B=ScalarParameter() {defaultValue=1, parameterName=texcoord})
  n12 = MaterialFunctionCall(Speed=ScalarParameter() {defaultValue=1, parameterName=speed02}) {materialFunction=&TimeWithSpeedVariable}
  n13 = MaterialFunctionCall(X=@n9.Speed X Tex2, Y=@n9.Speed X Tex2) {materialFunction=&MakeFloat2}
  n14 = Add(
   A=TextureSample(UVs=Panner(Coordinate=@n7, Time=@n8.Time, Speed=@n10.Result) {speedX=-1}) {texture=/DealershipSharedLibrary/FX/Textures/Gradient_Masks/T_VCVFX_Boost_Helix_WindParticles_Alpha}.R
   B=TextureSample(UVs=Panner(Coordinate=@n11, Time=@n12.Time, Speed=@n13.Result) {speedX=-3}) {texture=/DealershipSharedLibrary/FX/Textures/Gradient_Masks/T_VCVFX_speedline_A}.R
  )
  n15 = ComponentMask(None=TextureCoordinate()) {r=true}
  n16 = OneMinus(None=@n15)
  n17 = StaticSwitchParameter(True=Multiply(A=@n15, B=@n16), False=@n16) {parameterName=IsWheelWrap?}
  MaterialExpressionConstantClamp_15 = Saturate(None=Multiply(A=NamedRerouteUsage(), B=Multiply(A=@n14, B=@n17)))
  NamedRerouteDeclaration(None=@MaterialExpressionConstantClamp_15) {name=Trail 3, nodeColor=<rgba 0.8984 1 0 1>}
  n18 = NamedRerouteUsage()
  n19 = Multiply(A=NamedRerouteUsage(), B=Power(Base=NamedRerouteUsage(), Exp=Constant() {r=0.8}))
  MaterialExpressionConstantClamp_12 = Saturate(None=Multiply(A=Multiply(A=@n18, B=@n19), B=NamedRerouteUsage()))
  NamedRerouteDeclaration(None=Multiply(A=@MaterialExpressionConstantClamp_12, B=NamedRerouteUsage())) {name=M1 Alpha, nodeColor=<rgba 1 0 0.5625 1>}
  n20 = NamedRerouteUsage()
  NamedRerouteDeclaration(None=Multiply(A=DynamicParameter().R, B=Multiply(A=@n20, B=@n19))) {name=M1 RGB, nodeColor=<rgba 0 0.8828 1 1>}
  n21 = TextureCoordinate()
  MaterialExpressionConstantClamp_13 = Saturate(None=Multiply(A=ComponentMask(None=@n21) {r=true}, B=Constant() {r=20}))
  n22 = ComponentMask(None=TextureCoordinate()) {r=true}
  MaterialExpressionConstantClamp_14 = Saturate(None=Multiply(A=OneMinus(None=@n22), B=Constant() {r=10}))
  n23 = StaticSwitchParameter(True=Constant() {r=4}, False=Constant() {r=2}) {parameterName=IS MAIN RIBBON?}
  NamedRerouteDeclaration(None=Multiply(
    A=Multiply(A=@MaterialExpressionConstantClamp_13, B=@MaterialExpressionConstantClamp_13)
    B=Power(Base=@MaterialExpressionConstantClamp_14, Exp=@n23)
   )) {name=Fade 1, nodeColor=<rgba 0 0.0391 1 1>}
  n24 = TextureCoordinate()
  n25 = OneMinus(None=StaticSwitchParameter(True=ComponentMask(None=@n24) {r=true}, False=ComponentMask(None=@n24) {g=true}) {parameterName=IS MAIN RIBBON?})
  n26 = StaticSwitchParameter(True=Constant() {r=1.2}, False=Constant() {r=3}) {parameterName=IS MAIN RIBBON?}
  n27 = Constant() {r=2}
  n28 = Constant() {r=5}
  NamedRerouteDeclaration(None=Multiply(A=Power(Base=@n25, Exp=@n26) {constExponent=1.2}, B=StaticSwitchParameter(True=@n27, False=@n28) {parameterName=IS MAIN RIBBON?}) {constB=2}) {name=Fade2, nodeColor=<rgba 0.3594 1 0 1>}
  n29 = Panner(Coordinate=TextureCoordinate() {uTiling=2}) {speedX=-1.2}
  n30 = NamedRerouteUsage()
  NamedRerouteDeclaration(None=TextureSample(UVs=StaticSwitchParameter(True=@n29, False=@n30) {parameterName=IS MAIN RIBBON?}) {texture=/DealershipSharedLibrary/FX/Textures/TexturePackage/T_VCVFX_SmokeTrail01_RGB}.R) {name=Texture 1, nodeColor=<rgba 1 0.9609 0 1>}
  n31 = Panner(Coordinate=TextureCoordinate() {uTiling=0.8, vTiling=0.8}) {speedX=-0.75, speedY=0.1}
  n32 = Panner(Coordinate=TextureCoordinate() {uTiling=0.5, vTiling=0.5}) {speedX=-0.8, speedY=0.2}
  NamedRerouteDeclaration(None=TextureSample(UVs=StaticSwitchParameter(True=@n31, False=@n32) {parameterName=IS MAIN RIBBON?}) {texture=/DealershipSharedLibrary/FX/Textures/Noise/T_VCVFX_CloudDistort_A}.RGB) {name=Texture 2, nodeColor=<rgba 0.6406 0 1 1>}
  n33 = TextureCoordinate()
  n34 = MaterialFunctionCall(X=Constant() {r=0.5}, Y=Constant() {r=-0.1}) {materialFunction=&MakeFloat2}
  n35 = Panner(Coordinate=TextureCoordinate() {uTiling=0.75, vTiling=0.4}) {speedX=-2}
  n36 = MaterialFunctionCall(X=Constant() {r=1}, Y=ScalarParameter() {defaultValue=1, parameterName=Y SCALE FROM RIBBON 1}) {materialFunction=&MakeFloat2}
  n37 = ComponentMask(None=TextureSample(UVs=MaterialFunctionCall(UVs=@n35, Texture Scale=@n36.Result) {materialFunction=&ScaleUVsByCenter}.UVs) {texture=/DealershipSharedLibrary/FX/Textures/Noise/T_VCVFX_TrailMask_RGB}.R) {g=true, r=true}
  n38 = Constant() {r=0.5}
  n39 = Panner(Coordinate=Add(A=Add(A=@n33, B=@n34.Result), B=Multiply(A=@n37, B=@n38))) {speedX=-0.5}
  n40 = MaterialFunctionCall(X=Constant() {r=1}, Y=ScalarParameter() {defaultValue=1, parameterName=Y SCALE FROM RIBBON 1}) {materialFunction=&MakeFloat2}
  NamedRerouteDeclaration(None=Multiply(
    A=MaterialFunctionCall(UVs=@n39, Texture Scale=@n40.Result) {materialFunction=&ScaleUVsByCenter}.UVs
    B=ScalarParameter() {defaultValue=1, parameterName=texcoord}
   )) {name=Distortion Trail 2, nodeColor=<rgba 0.8047 1 0 1>}
  n41 = NamedRerouteUsage()
  NamedRerouteDeclaration(None=Multiply(
    A=DynamicParameter() {parameterIndex=1, defaultValue=<rgba 1 1 0 0>, paramNames [POWER,DISTORSION,OFFSETX,OFFSETY]}.POWER
    B=Multiply(A=@n41, B=@n3)
   )) {name=Trail 2 RGB, nodeColor=<rgba 1 0 0.8906 1>}
  n42 = NamedRerouteUsage()
  n43 = NamedRerouteUsage()
  n44 = MaterialFunctionCall(
   Opacity In=StaticSwitchParameter(True=NamedRerouteUsage(), False=StaticSwitchParameter(True=@n42, False=@n43) {parameterName=IS MAIN RIBBON?}) {parameterName=IS WINDTRAIL?}
   DepthFade Distance=ScalarParameter() {defaultValue=50, parameterName=DepthFade distance}
  ) {materialFunction=&MF_VCVFX_DepthFade}
  StaticSwitchParameter(
   True=Multiply(
    A=@n44.Result
    B=MaterialFunctionCall!(Multiply, Power EXP) {materialFunction=/DealershipSharedLibrary/FX/Materials/MaterialFunctions/MF_VCVFX_WheelwrapMask}.Result
   )
   False=@n44.Result
  ) {parameterName=IsWheelWrap?}
 ]
}
